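; Season 19 battle pass parent quest ability that grants gameplay effects to its owner as an objective is completed.
; On objective completion it calls the parent, then looks up the completion count in a count-to-effect map and applies the matching effect.
; An unmatched count in a non-empty map only prints "Objective not found".
; The activate and end events are disabled placeholder stubs.
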(event Ability|EventActivateAbility
  (|CallParentFunction) ; This node is disabled and will not be called.\nDrag off pins to build functionality.
)

(event Ability|EventOnEndAbility
  (params (bWasCancelled :type Boolean))
  (|CallParentFunction bWasCancelled) ; This node is disabled and will not be called.\nDrag off pins to build functionality.
)

(fn OnObjectiveCompleted
  (params (ObjectiveCompletionCount :type Integer) (ObjectiveBackendName :type Name :default "None"))
  (bind _completioncounttoges (Variables|Default|GetCompletionCountToGEs))
  (bind (_value _returnvalue) (Utilities|Map|Find _completioncounttoges ObjectiveCompletionCount))
  (|CallParentFunction ObjectiveCompletionCount)
  (if _returnvalue
    (Ability|ApplyGameplayEffectToOwner _value)
    (elif (Utilities|Map|IsNotEmpty _completioncounttoges)
      (Development|PrintString "Objective not found"))))
