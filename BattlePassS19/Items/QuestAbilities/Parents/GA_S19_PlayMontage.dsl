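; Parent gameplay ability for Season 19 Battle Pass quest items that plays a device-equip montage on the player pawn.
; Activation calls the parent, starts the montage task and hides the pawn's current weapon.
; The ability's end event unhides whichever weapon is current at that moment.
; The montage's completion, blend-out, interrupt and cancel outputs carry no logic.
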(event Ability|EventActivateAbility
  (|CallParentFunction)
  (Ability|Tasks|PlayMontageAndWait "None" "/Game/Animation/Game/MainPlayer/Combat/Gadgets/Medium/Male/Device/Device_Equip_MMC_M.Device_Equip_MMC_M"
    (:then
      (bind _asfort_player_pawn (Utilities|Casting|CastToFortPlayerPawn (Ability|GetAvatarActorfromActorInfo)))
      (Rendering|SetActorHiddenInGame (Variables|/Script/FortniteGame.FortPawn|GetCurrentWeapon _asfort_player_pawn) true))
    (:OnCompleted
      _)
    (:OnBlendOut
      _)
    (:OnInterrupted
      _)
    (:OnCancelled
      _))
  (Ability|EndAbility))

(event Ability|EventFailedToActivatePassiveAbility
  (params (ActorInfo :type Gameplay%20Ability%20Actor%20Info%20Structure)))

(event Ability|EventOnEndAbility
  (params (bWasCancelled :type Boolean))
  (bind _asfort_player_pawn (Utilities|Casting|CastToFortPlayerPawn (Ability|GetAvatarActorfromActorInfo)))
  (Rendering|SetActorHiddenInGame (Variables|/Script/FortniteGame.FortPawn|GetCurrentWeapon _asfort_player_pawn) false))
