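; Deprecated parent gameplay ability for Season 19 quest voice-over, playing a sound sequence on the avatar.
; On authority, it applies a VO-playing effect whose stack count serves as a queue position, decremented as stacks drop.
; After a net sync and optional delay, it starts a looping sound and chains sequence items through audio-finished callbacks and per-item delays.
; An invalid item ends the ability, which releases its stack and stops the loop.
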(event Ability|EventActivateAbility
  (bind _playdelay (Variables|Default|GetPlayDelay))
  (bind _placeinline (Variables|Default|GetPlaceInLine))
  (|CallParentFunction)
  (Ability|HasAuthority
    (:True
      (bind _returnvalue (Ability|ApplyGameplayEffectToOwner "/BattlePassS19/Items/QuestAbilities/Parents/GE_S19_PlayVO_VOPlaying_DEPRICATED.GE_S19_PlayVO_VOPlaying_DEPRICATED_C"))
      (bind _output_get (Variables|Default|SetGEVOPlayingHandle _returnvalue))
      (bind _returnvalue_2 (Ability|GameplayEffect|GetActiveGameplayEffectStackCount _output_get))
      (bind _output_get_2 (Variables|Default|SetPlaceInLine _returnvalue_2))
      (if (> _output_get_2 1)
        (bind (_asynctaskproxy _handle _newcount _oldcount) (Ability|Tasks|WaitForGameplayEffectStackChange _output_get)
          (:OnChange
            (if (< _newcount _oldcount)
              (Math|Integer|DecrementInt _placeinline
                (:%20%20
                  (if (<= _placeinline 1)
                    (GameplayTasks|EndTask _asynctaskproxy)
                    (Utilities|FlowControl|DoOnce false
                      (:Completed
                        _))))))))
        (else
          _))
      (Ability|Tasks|WaitNetSync "BothWait"
        (:OnSync
          (if (<= _playdelay 0)
            (CallFunction|HandleVO)
            (else
              (Utilities|Time|SetTimerbyEvent (EventDispatchers|CreateEvent :!event Custom|HandleVO) _playdelay))))))
    (:False
      _)))

(event Ability|EventFailedToActivatePassiveAbility
  (params (ActorInfo :type Gameplay%20Ability%20Actor%20Info%20Structure)))

(event Custom|HandleVO
  (Default|Play)
  (bind _returnvalue (Audio|SpawnSoundAttached :Sound (Variables|Default|GetLoopingSound) :AttachToComponent (Variables|/Script/Engine.Actor|GetRootComponent (Ability|GetAvatarActorfromActorInfo)) :AttachPointName "None" :LocationType "KeepRelativeOffset" :bStopWhenAttachedToDestroyed false :VolumeMultiplier 1.0 :PitchMultiplier 1.0 :StartTime 0.0 :bAutoDestroy false))
  (Variables|Default|SetLoopingSoundHandle _returnvalue))

(event Ability|EventOnEndAbility
  (params (bWasCancelled :type Boolean))
  (bind _loopingsoundhandle (Variables|Default|GetLoopingSoundHandle))
  (Ability|RemoveGameplayEffectFromOwnerWithAssetTags (Variables|Default|GetAssetTags))
  (Ability|RemoveGameplayEffectFromOwnerWithHandle (Variables|Default|GetGEVOPlayingHandle) 1)
  (Audio|Components|Audio|Stop _loopingsoundhandle)
  (Components|DestroyComponent _loopingsoundhandle))

(fn AdvanceSequence
  (Math|Integer|IncrementInt (Variables|Default|GetCurrentIdx)
    (:%20%20
      (Default|PlayCurrentItem))))

(fn GetCurrentItem
  (return (Utilities|Array|Get(acopy) (Variables|Default|GetSoundSequence) (Variables|Default|GetCurrentIdx))))

(fn Play
  (Networking|CanExecuteCosmeticEvents
    (:True
      (Default|PlayCurrentItem))))

(fn PlayCurrentItem
  (bind _output_sound_2_d8314be3444a737f8a2ed3b3edeac6b5 (Utilities|Struct|BreakSoundConcatenatorSequenceItemData (Default|GetCurrentItem)))
  (Utilities|IsValid _output_sound_2_d8314be3444a737f8a2ed3b3edeac6b5
    (:Is%20Valid
      (bind _returnvalue (Audio|SpawnSoundAttached _output_sound_2_d8314be3444a737f8a2ed3b3edeac6b5 (Variables|/Script/Engine.Actor|GetRootComponent (Ability|GetAvatarActorfromActorInfo))))
      (EventDispatchers|BindEventtoOnAudioFinished _returnvalue (EventDispatchers|CreateEvent :!event StartDelay)))
    (:Is%20Not%20Valid
      (Ability|EndAbility))))

(fn StartDelay
  (bind _output (Default|GetCurrentItem))
  (bind (_sound_2_d8314be3444a737f8a2ed3b3edeac6b5 _delay_5_9b967cfa4e4664110f1201a1b0291d30) (Utilities|Struct|BreakSoundConcatenatorSequenceItemData _output))
  (if (== _delay_5_9b967cfa4e4664110f1201a1b0291d30 0.0)
    (Default|AdvanceSequence)
    (else
      (Utilities|Time|SetTimerbyEvent (EventDispatchers|CreateEvent :!event AdvanceSequence) _delay_5_9b967cfa4e4664110f1201a1b0291d30))))
